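# Material function that stamps a tinted banner icon on the front and back of the TacticKale player body's vest and shirt.
# It offsets UV1 per placement, using separate UV tiles and step-built rectangle masks to isolate each region.
# Center-scaled front and back UVs are blended into one BannerTex sample.
# Its red channel, gated by masks and the isShirtOnly parameter, lerps incoming BaseColor toward BannerColor.
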MaterialFunction /BRCosmetics/Characters/Player/Male/Medium/Bodies/M_MED_TacticKale/Materials/MF_TacticBannerIcon
graph {nodes [
  Add(A=NamedRerouteUsage())
  n1 = GetMaterialAttributes(None=FunctionInput() {inputType=FunctionInput_MaterialAttributes}) {attributeGetTypes [#1]}
  n2 = VectorParameter() {group=TacticCustomize, parameterName=BannerColor, sortPriority=60, defaultValue=<rgba 1 1 1 1>}
  n3 = ScalarParameter() {defaultValue=0.6, group=TacticCustomize, parameterName=Back_BannerIconSizeX, sortPriority=65}
  n4 = ScalarParameter() {defaultValue=0.6, group=TacticCustomize, parameterName=Back_BannerIconSizeY, sortPriority=65}
  n5 = MaterialFunctionCall(UVs=Frac(None=NamedRerouteUsage()), Texture Scale=AppendVector(A=@n3, B=@n4)) {materialFunction=&ScaleUVsByCenter}
  n6 = MaterialFunctionCall(UVs=Frac(None=NamedRerouteUsage()), Texture Scale=AppendVector(A=@n3, B=@n4)) {materialFunction=&ScaleUVsByCenter}
  n7 = NamedRerouteUsage()
  n8 = Frac(None=NamedRerouteUsage())
  n9 = ScalarParameter() {defaultValue=0.4, group=TacticCustomize, parameterName=Front_BannerIconSize, sortPriority=61}
  n10 = TextureSampleParameter2D(UVs=LinearInterpolate(
    A=LinearInterpolate(A=@n5.UVs, B=@n6.UVs, Alpha=@n7)
    B=MaterialFunctionCall(UVs=@n8, Texture Scale=@n9) {materialFunction=&ScaleUVsByCenter}.UVs
    Alpha=NamedRerouteUsage()
   ) {constA=0.5, constAlpha=1}) {
   group=TacticCustomize
   parameterName=BannerTex
   samplerSource=SSM_Wrap_WorldGroupSettings
   sortPriority=60
   texture=/Game/UI/Foundation/Textures/Banner/Shapes/banner_shape_standard_Mipmaps
  }
  n11 = LinearInterpolate(A=NamedRerouteUsage(), B=NamedRerouteUsage(), Alpha=NamedRerouteUsage())
  n12 = NamedRerouteUsage()
  n13 = Add(A=ScalarParameter() {group=TacticCustomize, parameterName=isShirtOnly, sortPriority=69}, B=NamedRerouteUsage())
  n14 = LinearInterpolate(B=@n10.R, Alpha=LinearInterpolate(B=Add(A=@n11, B=@n12), Alpha=Saturate(None=@n13))) {constAlpha=1}
  FunctionOutput(None=SetMaterialAttributes!(MaterialAttributes=@n1.MaterialAttributes, BaseColor=LinearInterpolate(A=@n1.BaseColor, B=@n2.RGB, Alpha=@n14)) {attributeSetTypes [#1]})
  n15 = Constant()
  n16 = ScalarParameter() {defaultValue=0.055, group=TacticCustomize, parameterName=Back_BannerPosVest, sortPriority=67}
  NamedRerouteDeclaration(None=Add(A=TextureCoordinate() {coordinateIndex=1}, B=AppendVector(A=@n15, B=@n16)) {constB=0.1}) {name=UV_Back_Vest, nodeColor=<rgba 1 0.6563 0 1>}
  n17 = ScalarParameter() {defaultValue=0.301, group=TacticCustomize, parameterName=Front_BannerIconMask, sortPriority=62}
  NamedRerouteDeclaration(None=@n17) {name=MinMask_Front, nodeColor=<rgba 1 0 0.0938 1>}
  n18 = NamedRerouteUsage()
  n19 = Step(Y=NamedRerouteUsage(), X=@n18) {constX=0, constY=1.26}
  n20 = ComponentMask(None=@n19) {r=true}
  n21 = ComponentMask(None=@n19) {g=true}
  n22 = Step(Y=@n18, X=NamedRerouteUsage()) {constX=1.74, constY=1}
  n23 = ComponentMask(None=@n22) {r=true}
  n24 = ComponentMask(None=@n22) {g=true}
  NamedRerouteDeclaration(None=Min(A=Min(A=@n20, B=@n21), B=Min(A=@n23, B=@n24))) {name=ShirtFMask, nodeColor=<rgba 0.4531 1 0 1>}
  n25 = NamedRerouteUsage()
  n26 = ComponentMask(None=Step(Y=NamedRerouteUsage(), X=@n25) {constX=0, constY=1.26}) {r=true}
  n27 = ComponentMask(None=Step(Y=@n25, X=NamedRerouteUsage()) {constX=1.74, constY=1}) {r=true}
  n28 = NamedRerouteUsage()
  n29 = ComponentMask(None=Step(Y=NamedRerouteUsage(), X=@n28) {constX=0, constY=1.26}) {g=true}
  n30 = ComponentMask(None=Step(Y=@n28, X=NamedRerouteUsage()) {constX=1.74, constY=1}) {g=true}
  NamedRerouteDeclaration(None=Min(A=Min(A=@n26, B=@n27), B=Min(A=@n29, B=@n30))) {name=ShirtBMask, nodeColor=<rgba 0 1 0.0625 1>}
  n31 = ScalarParameter() {defaultValue=0.205, group=TacticCustomize, parameterName=Back_BannerIconMaskX, sortPriority=66}
  NamedRerouteDeclaration(None=OneMinus(None=@n31)) {name=MaxMask_Back_X, nodeColor=<rgba 0.0312 1 0 1>}
  n32 = TextureCoordinate() {coordinateIndex=1}
  n33 = Constant2Vector() {g=1.1, r=1}
  n34 = Constant()
  n35 = ScalarParameter() {group=TacticCustomize, parameterName=Front_BannerPos, sortPriority=63}
  NamedRerouteDeclaration(None=Add(A=Multiply(A=@n32, B=@n33), B=AppendVector(A=@n34, B=@n35)) {constB=0.1}) {name=UV_Front, nodeColor=<rgba 1 0.6562 0 1>}
  n36 = NamedRerouteUsage()
  n37 = Step(X=@n36) {constX=0}
  n38 = Min(A=ComponentMask(None=@n37) {r=true}, B=ComponentMask(None=@n37) {g=true})
  n39 = Step(Y=@n36) {constY=1}
  n40 = Min(A=ComponentMask(None=@n39) {r=true}, B=ComponentMask(None=@n39) {g=true})
  n41 = NamedRerouteUsage()
  n42 = Step(X=@n41) {constX=0}
  n43 = Min(A=ComponentMask(None=@n42) {r=true}, B=ComponentMask(None=@n42) {g=true})
  n44 = Step(Y=@n41) {constY=1}
  n45 = Min(A=ComponentMask(None=@n44) {r=true}, B=ComponentMask(None=@n44) {g=true})
  NamedRerouteDeclaration(None=Add(A=Min(A=@n38, B=@n40), B=Min(A=@n43, B=@n45))) {name=VestBannerLerp, nodeColor=<rgba 0.3594 1 0 1>}
  NamedRerouteDeclaration(None=Step(X=VertexColor().R) {constX=0, constY=0.1}) {name=VertexR, nodeColor=<rgba 1 0.3281 0 1>}
  n46 = ScalarParameter() {defaultValue=0.205, group=TacticCustomize, parameterName=Back_BannerIconMaskY, sortPriority=66}
  NamedRerouteDeclaration(None=@n46) {name=MinMask_Back_Y, nodeColor=<rgba 1 0 0.0938 1>}
  NamedRerouteDeclaration(None=OneMinus(None=@n46)) {name=MaxMask_Back_Y, nodeColor=<rgba 0.0312 1 0 1>}
  n47 = Constant()
  n48 = ScalarParameter() {defaultValue=0.055, group=TacticCustomize, parameterName=Back_BannerPosShirt, sortPriority=68}
  NamedRerouteDeclaration(None=Add(A=TextureCoordinate() {coordinateIndex=1}, B=AppendVector(A=@n47, B=@n48)) {constB=0.1}) {name=UV_Back_Shirt, nodeColor=<rgba 1 0.6562 0 1>}
  NamedRerouteDeclaration(None=OneMinus(None=@n17)) {name=MaxMask_Front, nodeColor=<rgba 0.0312 1 0 1>}
  n49 = NamedRerouteUsage()
  n50 = ComponentMask(None=@n49) {r=true}
  n51 = ComponentMask(None=@n49) {g=true}
  NamedRerouteDeclaration(None=AppendVector(A=Subtract(A=@n50), B=Subtract(A=@n51))) {name=VestF, nodeColor=<rgba 0 0.2734 1 1>}
  n52 = NamedRerouteUsage()
  n53 = ComponentMask(None=@n52) {r=true}
  n54 = ComponentMask(None=@n52) {g=true}
  NamedRerouteDeclaration(None=AppendVector(A=Subtract(A=@n53) {constB=2}, B=Subtract(A=@n54) {constB=0})) {name=VestB, nodeColor=<rgba 0 0.7187 1 1>}
  n55 = NamedRerouteUsage()
  n56 = ComponentMask(None=@n55) {r=true}
  n57 = ComponentMask(None=@n55) {g=true}
  NamedRerouteDeclaration(None=AppendVector(A=Subtract(A=@n56), B=Subtract(A=@n57) {constB=0})) {name=ShirtF, nodeColor=<rgba 0 1 0.1094 1>}
  n58 = NamedRerouteUsage()
  n59 = ComponentMask(None=@n58) {r=true}
  n60 = ComponentMask(None=@n58) {g=true}
  NamedRerouteDeclaration(None=AppendVector(A=Subtract(A=@n59) {constB=2}, B=Subtract(A=@n60) {constB=0})) {name=ShirtB, nodeColor=<rgba 0 1 0.5078 1>}
  NamedRerouteDeclaration(None=@n31) {name=MinMask_Back_X, nodeColor=<rgba 1 0 0.0937 1>}
  n61 = NamedRerouteUsage()
  n62 = Step(Y=NamedRerouteUsage(), X=@n61) {constX=0, constY=1.26}
  n63 = ComponentMask(None=@n62) {r=true}
  n64 = ComponentMask(None=@n62) {g=true}
  n65 = Step(Y=@n61, X=NamedRerouteUsage()) {constX=1.74, constY=1}
  n66 = ComponentMask(None=@n65) {r=true}
  n67 = ComponentMask(None=@n65) {g=true}
  NamedRerouteDeclaration(None=Min(A=Min(A=@n63, B=@n64), B=Min(A=@n66, B=@n67))) {name=VestFMask, nodeColor=<rgba 0 1 0.7891 1>}
  n68 = NamedRerouteUsage()
  n69 = ComponentMask(None=Step(Y=NamedRerouteUsage(), X=@n68) {constX=0, constY=1.26}) {r=true}
  n70 = ComponentMask(None=Step(Y=@n68, X=NamedRerouteUsage()) {constX=1.74, constY=1}) {r=true}
  n71 = NamedRerouteUsage()
  n72 = ComponentMask(None=Step(Y=NamedRerouteUsage(), X=@n71) {constX=0, constY=1.26}) {g=true}
  n73 = ComponentMask(None=Step(Y=@n71, X=NamedRerouteUsage()) {constX=1.74, constY=1}) {g=true}
  NamedRerouteDeclaration(None=Min(A=Min(A=@n69, B=@n70), B=Min(A=@n72, B=@n73))) {name=VestBMask, nodeColor=<rgba 1 0.7266 0 1>}
 ]}
