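# Material function that recolors the BaseColor of the TacticKale male character body through three zone masks, with camo and tattoo layers.
# Zone colors, switchable to test colors, feed a chain of lerps that writes BaseColor back into the incoming material attributes.
# A scaled, offset camo texture on selectable UVs supplies an optional overlay blend.
# Tattoos sample a selectable UV channel, with Step and Min masks clipping them to the 0–1 range.
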MaterialFunction /BRCosmetics/Characters/Player/Male/Medium/Bodies/M_MED_TacticKale/Materials/MF_TacticKale_ColorCustomize
graph {nodes [
  n1 = NamedRerouteUsage()
  n2 = LinearInterpolate(
   A=TextureCoordinate()
   B=TextureCoordinate() {coordinateIndex=1}
   Alpha=ScalarParameter() {defaultValue=1, group=TacticCustomize, parameterName="TC_UseSecondUVChannel(Tattoo)", sortPriority=52}
  )
  n3 = TextureSampleParameter2D(UVs=@n2) {group=TacticCustomize, parameterName=TatoosTex, sortPriority=50, texture=&T_BlacksRGB}
  n4 = ScalarParameter() {group=TacticCustomize, parameterName=TattooOpacity, sortPriority=51}
  n5 = Step(X=@n2) {constX=0}
  n6 = ComponentMask(None=@n5) {r=true}
  n7 = ComponentMask(None=@n5) {g=true}
  n8 = Step(Y=@n2) {constY=1}
  n9 = ComponentMask(None=@n8) {r=true}
  n10 = ComponentMask(None=@n8) {g=true}
  n11 = Multiply(
   A=Multiply(
    A=Multiply(A=@n3.A, B=@n4)
    B=ScalarParameter() {defaultValue=0.92, group=TacticCustomize, parameterName=TattooColorIntensity, sortPriority=52}
   )
   B=Min(A=Min(A=@n6, B=@n7), B=Min(A=@n9, B=@n10))
  )
  n12 = NamedRerouteUsage()
  n13 = LinearInterpolate(
   A=NamedRerouteUsage()
   B=MaterialFunctionCall(Base=NamedRerouteUsage(), Blend=NamedRerouteUsage()) {materialFunction=&Blend_Overlay}.Result
   Alpha=ScalarParameter() {defaultValue=1, group=TacticCustomize, parameterName=useCamo, sortPriority=29}
  )
  n14 = LinearInterpolate(
   A=LinearInterpolate(A=LinearInterpolate(A=@n1, B=@n3.RGB, Alpha=@n11) {constA=1, constAlpha=1}, B=Multiply(A=@n12, B=@n13), Alpha=NamedRerouteUsage())
   B=Multiply(A=NamedRerouteUsage(), B=NamedRerouteUsage())
   Alpha=NamedRerouteUsage()
  )
  n15 = Multiply(A=NamedRerouteUsage(), B=NamedRerouteUsage())
  n16 = NamedRerouteUsage()
  FunctionOutput(None=SetMaterialAttributes!(MaterialAttributes=NamedRerouteUsage(), BaseColor=LinearInterpolate(A=@n14, B=@n15, Alpha=@n16)) {attributeSetTypes [#1]})
  n17 = GetMaterialAttributes(None=FunctionInput() {inputType=FunctionInput_MaterialAttributes}) {attributeGetTypes [#1]}
  NamedRerouteDeclaration(None=@n17.MaterialAttributes) {name=MA, nodeColor=<rgba 0 1 0.9531 1>}
  NamedRerouteDeclaration(None=@n17.BaseColor) {name=BaseColor, nodeColor=<rgba 0.2188 0 1 1>}
  n18 = MaterialFunctionCall(Float3=FunctionInput() {inputName=M, sortPriority=1}) {materialFunction=&BreakOutFloat3Components}
  NamedRerouteDeclaration(None=@n18.B) {name=MSkin, nodeColor=<rgba 0.1484 1 0 1>}
  n19 = MaterialFunctionCall(Float3=FunctionInput() {inputName=TAMask, sortPriority=2}) {materialFunction=&BreakOutFloat3Components}
  NamedRerouteDeclaration(None=@n19.R) {name=Zone1Mask, nodeColor=<rgba 0 1 0.6484 1>}
  NamedRerouteDeclaration(None=@n19.G) {name=Zone2Mask, nodeColor=<rgba 0.4062 0 1 1>}
  NamedRerouteDeclaration(None=@n19.B) {name=Zone3Mask, nodeColor=<rgba 1 0 0.2344 1>}
  NamedRerouteDeclaration(None=Add(A=@n18.G, B=ScalarParameter() {group=TacticCustomize, parameterName=OverlayAdding, sortPriority=1})) {name=OverlayBlend, nodeColor=<rgba 1 0 0.7031 1>}
  n20 = VectorParameter() {group=TacticCustomize, parameterName=Test_Zone1Color, sortPriority=100, defaultValue=<rgba 0 0 0 1>}
  n21 = StaticBoolParameter() {group=TacticCustomize, parameterName=TestColors, sortPriority=99}
  NamedRerouteDeclaration(None=StaticSwitch(
    True=@n20.RGB
    False=VectorParameter() {group=TacticCustomize, parameterName=Zone1Color, sortPriority=10, defaultValue=<rgba 0 0 0 1>}.RGB
    Value=@n21
   )) {name=Zone1Color, nodeColor=<rgba 1 0 0.0446 1>}
  n22 = VectorParameter() {group=TacticCustomize, parameterName=Test_Zone2Color, sortPriority=100, defaultValue=<rgba 0 0 0 1>}
  NamedRerouteDeclaration(None=StaticSwitch(
    True=@n22.RGB
    False=VectorParameter() {group=TacticCustomize, parameterName=Zone2Color, sortPriority=11, defaultValue=<rgba 0 0 0 1>}.RGB
    Value=@n21
   )) {name=Zone2Color, nodeColor=<rgba 0 1 0 1>}
  n23 = VectorParameter() {group=TacticCustomize, parameterName=Test_Zone3Color, sortPriority=100, defaultValue=<rgba 0 0 0 1>}
  NamedRerouteDeclaration(None=StaticSwitch(
    True=@n23.RGB
    False=VectorParameter() {group=TacticCustomize, parameterName=Zone3Color, sortPriority=12, defaultValue=<rgba 0 0 0 1>}.RGB
    Value=@n21
   )) {name=Zone3Color, nodeColor=<rgba 0 0 1 1>}
  n24 = LinearInterpolate(
   A=TextureCoordinate()
   B=TextureCoordinate() {coordinateIndex=2}
   Alpha=ScalarParameter() {group=TacticCustomize, parameterName=isCamoUV2, sortPriority=29}
  )
  n25 = ScalarParameter() {defaultValue=0.5, group=TacticCustomize, parameterName=CamuTilling, sortPriority=31}
  n26 = ScalarParameter() {group=TacticCustomize, parameterName=CamuOffsetX, sortPriority=31}
  n27 = ScalarParameter() {group=TacticCustomize, parameterName=CamuOffsetY, sortPriority=31}
  n28 = TextureSampleParameter2D(UVs=Add(
    A=MaterialFunctionCall(UVs=@n24, Texture Scale=@n25) {materialFunction=&ScaleUVsByCenter}.UVs
    B=AppendVector(A=@n26, B=@n27)
   )) {
   group=TacticCustomize
   parameterName=CamuTex
   sortPriority=30
   texture=/Game/Characters/Player/Female/Medium/Bodies/F_MED_Desert_Ops_Camo/Textures/Base/Camo/T_Desert_Ops_Camo_Arctic_Tiling
  }
  NamedRerouteDeclaration(None=Add(A=@n28.R, B=ScalarParameter() {group=TacticCustomize, parameterName=CamoAdding, sortPriority=30})) {name=TexCamo, nodeColor=<rgba 1 0 0.7031 1>}
  n29 = VectorParameter() {group=TacticCustomize, parameterName=TattooColor, sortPriority=56, defaultValue=<rgba 0.03 0.0111 0.0081 1>}
  StaticSwitchParameter(True=@n29.RGB, False=@n3.RGB) {defaultValue=true, group=TacticCustomize, parameterName=TattooColorSwitch, sortPriority=55}
 ]}
