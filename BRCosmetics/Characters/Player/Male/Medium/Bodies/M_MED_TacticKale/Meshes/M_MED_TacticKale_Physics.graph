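# Physics asset for the TacticKale medium male body cosmetic, defining secondary motion for left and right weapon props and front and back grenades with rings.
# Small base capsules tie to a pelvis capsule, with weapon bases angularly locked and grenade bases angularly free.
# Each weapon and grenade swings within narrow limits on its base.
# Ring bases lock to their grenades, letting each ring swing widely on one axis.
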PhysicsAsset /BRCosmetics/Characters/Player/Male/Medium/Bodies/M_MED_TacticKale/Meshes/M_MED_TacticKale_Physics
bodies {
 dyn_grenade_bk [{shapeType=Capsule, length=4, radius=2, shapeName=dyn_grenade_bk_capsule, center=<-6.964 0 0>, rotation=<rot 0.0002 -89.9995 -89.9996>}]
 dyn_grenade_bk_base [{shapeType=Capsule, length=2, radius=1, shapeName=dyn_grenade_bk_base_capsule}]
 dyn_grenade_fr [{shapeType=Capsule, length=4, radius=2, shapeName=dyn_grenade_fr_capsule, center=<-6.964 0 0>, rotation=<rot 90 0 0>}]
 dyn_grenade_fr_base [{shapeType=Capsule, length=2, radius=1, shapeName=dyn_grenade_fr_base_capsule}]
 dyn_grenade_ring_bk [{shapeType=Capsule, length=5, radius=1, shapeName=dyn_grenade_ring_bk_capsule, center=<1.324 0 0>, rotation=<rot 90 0 0>}]
 dyn_grenade_ring_bk_base [{shapeType=Capsule, length=0.505, radius=0.505, shapeName=dyn_grenade_ring_bk_base_capsule, rotation=<rot 0 -90 0>}]
 dyn_grenade_ring_fr [{shapeType=Capsule, length=5, radius=1, shapeName=dyn_grenade_ring_fr_capsule, center=<5 0 0>, rotation=<rot 90 0 0>}]
 dyn_grenade_ring_fr_base [{shapeType=Capsule, length=0.505, radius=0.505, shapeName=dyn_grenade_ring_fr_base_capsule, rotation=<rot 0 -90 0>}]
 dyn_weapon_base_l [{shapeType=Capsule, length=2, radius=2, shapeName=dyn_weapon_base_l_capsule, rotation=<rot 0 -90 0>}]
 dyn_weapon_base_r [{shapeType=Capsule, length=2, radius=2, shapeName=dyn_weapon_base_l_capsule, rotation=<rot 0 -90 180>}]
 dyn_weapon_l [{shapeType=Capsule, length=4, radius=5, shapeName=dyn_weapon_l_capsule, center=<4 0 0>}]
 dyn_weapon_r [{shapeType=Capsule, length=4, radius=5, shapeName=dyn_weapon_l_capsule, center=<4 0 0>}]
 pelvis [{shapeType=Capsule, length=0, radius=17.4827, center=<23.6195 -0.4387 1.0952>, rotation=<rot 5.3245 -90 3.7915>}]
}
constraints [
 {bone1Name=dyn_weapon_l, bone2Name=dyn_weapon_base_l, swing1LimitDegrees=10, swing2LimitDegrees=0.05, twistLimitDegrees=10}
 {bone1Name=dyn_weapon_r, bone2Name=dyn_weapon_base_r, swing1LimitDegrees=10, swing2LimitDegrees=0.05, twistLimitDegrees=10}
 {
  bone1Name=dyn_weapon_base_l
  bone2Name=pelvis
  swing1LimitDegrees=45
  swing1Motion=Locked
  swing2LimitDegrees=45
  swing2Motion=Locked
  twistLimitDegrees=45
  twistMotion=Locked
 }
 {
  bone1Name=dyn_weapon_base_r
  bone2Name=pelvis
  swing1LimitDegrees=45
  swing1Motion=Locked
  swing2LimitDegrees=45
  swing2Motion=Locked
  twistLimitDegrees=45
  twistMotion=Locked
 }
 {bone1Name=dyn_grenade_fr, bone2Name=dyn_grenade_fr_base, swing1LimitDegrees=5, swing2LimitDegrees=5, twistLimitDegrees=3}
 {
  bone1Name=dyn_grenade_fr_base
  bone2Name=pelvis
  swing1LimitDegrees=45
  swing1Motion=Free
  swing2LimitDegrees=45
  swing2Motion=Free
  twistLimitDegrees=45
  twistMotion=Free
 }
 {bone1Name=dyn_grenade_bk, bone2Name=dyn_grenade_bk_base, swing1LimitDegrees=10, swing2LimitDegrees=10, twistLimitDegrees=3}
 {
  bone1Name=dyn_grenade_bk_base
  bone2Name=pelvis
  swing1LimitDegrees=45
  swing1Motion=Free
  swing2LimitDegrees=45
  swing2Motion=Free
  twistLimitDegrees=45
  twistMotion=Free
 }
 {
  bone1Name=dyn_grenade_ring_fr_base
  bone2Name=dyn_grenade_fr
  swing1LimitDegrees=10
  swing1Motion=Locked
  swing2LimitDegrees=45
  swing2Motion=Locked
  twistLimitDegrees=5
  twistMotion=Locked
 }
 {bone1Name=dyn_grenade_ring_fr, bone2Name=dyn_grenade_ring_fr_base, swing1LimitDegrees=10, swing2LimitDegrees=65, twistLimitDegrees=5}
 {bone1Name=dyn_grenade_ring_bk, bone2Name=dyn_grenade_ring_bk_base, swing1LimitDegrees=10, swing2LimitDegrees=65, twistLimitDegrees=5}
 {
  bone1Name=dyn_grenade_ring_bk_base
  bone2Name=dyn_grenade_bk
  swing1LimitDegrees=45
  swing1Motion=Locked
  swing2LimitDegrees=45
  swing2Motion=Locked
  twistLimitDegrees=45
  twistMotion=Locked
 }
]
