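# Animation Blueprint for the TacticKale medium male player body cosmetic, layering secondary motion and corrections onto the pose copied from the parent mesh.
# The pose runs through a deform Control Rig, knee-pad constraint, pelvis-space RigidBody physics and weighted weapon-base constraints.
# A backpack-attach offset follows, enabled by a vest-off flag set from a cosmetic meta tag in the front end.
# A second Control Rig produces the final pose.
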AnimBlueprint /BRCosmetics/Characters/Player/Male/Medium/Bodies/M_MED_TacticKale/Meshes/M_MED_TacticKale_AnimBP
target_skeleton=&/M_MED_TacticKale_Skeleton
graphs [
 AnimGraph {nodes [
   n1 = Control Rig(Source=Copy Pose From Mesh() {bUseAttachedParent=true}) {
    alphaBoolBlend {blendOption=Linear}
    controlRigAssetReference {blueprintRigClass=&/M_MED_TacticKale_DeformRig.M_MED_TacticKale_DeformRig_C}
   }
   n2 = "Constraint - extra_knee_pad_l (PP)"!(ComponentPose=Local To Component(LocalPose=@n1), Alpha=1, ConstraintWeights_0=0.5, ConstraintWeights_1:float=0.5) -> (Pose) {
    alphaBoolBlend {blendOption=Linear}
    boneToModify {boneName=extra_knee_pad_l}
    constraintSetup [{targetBone {boneName=thigh_l}}, {targetBone {boneName=calf_l}}]
    constraintWeights [1,1]
   }
   n3 = RigidBody!(ComponentPose=@n2, Alpha, OverrideWorldGravity:Vector="0,0,-980", ExternalForce) -> (Pose) {
    bClampLinearTranslationLimitToRefPose=true
    bOverrideWorldGravity=true
    overridePhysicsAsset=&/M_MED_TacticKale_Physics
    simulationSpace=BaseBoneSpace
    alphaBoolBlend {blendOption=Linear}
    baseBoneRef {boneName=pelvis}
    componentAppliedLinearAccClamp=<3000 3000>
    componentLinearAccScale=<0.6 0.6 0.8>
    componentLinearVelScale=<0.6 0.6 0.8>
    overrideWorldGravity {z=-980}
    simSpaceSettings {maxAngularAcceleration=0, maxAngularVelocity=15000, maxLinearAcceleration=0, maxLinearVelocity=15000, worldAlpha=0.05}
   }
   n4 = "Constraint - dyn_weapon_base_l (PPPPP)"!(
    ComponentPose=@n3
    Alpha=1
    ConstraintWeights_0=0.0925
    ConstraintWeights_1:float=0.0634
    ConstraintWeights_2:float=0.7882
    ConstraintWeights_3:float=0.0096
    ConstraintWeights_4:float=0.0464
   ) -> (Pose) {
    alphaBoolBlend {blendOption=Linear}
    boneToModify {boneName=dyn_weapon_base_l}
    constraintSetup [
     {targetBone {boneName=pelvis}}
     {targetBone {boneName=thigh_l}}
     {targetBone {boneName=thigh_twist_01_l}}
     {targetBone {boneName=deform_glute_l}}
     {targetBone {boneName=deform_groin_l}}
    ]
    constraintWeights [1,1,1,1,1]
   }
   n5 = "Constraint - dyn_weapon_base_r (PPPPP)"!(
    ComponentPose=@n4
    Alpha=1
    ConstraintWeights_0=0.0929
    ConstraintWeights_1:float=0.0644
    ConstraintWeights_2:float=0.7857
    ConstraintWeights_3:float=0.0099
    ConstraintWeights_4:float=0.0471
   ) -> (Pose) {
    alphaBoolBlend {blendOption=Linear}
    boneToModify {boneName=dyn_weapon_base_r}
    constraintSetup [
     {targetBone {boneName=pelvis}}
     {targetBone {boneName=thigh_r}}
     {targetBone {boneName=thigh_twist_01_r}}
     {targetBone {boneName=deform_glute_r}}
     {targetBone {boneName=deform_groin_r}}
    ]
    constraintWeights [1,1,1,1,1]
   }
   n6 = "Transform (Modify) Bone - Bone: attach_backpack"!(ComponentPose=@n5, bAlphaBoolEnabled:bool=Get Vest Off() -> (Vest Off:bool), Translation="0,3,0", Rotation="0,0,0", Scale="1,1,1") -> (Pose) {
    alphaInputType=Bool
    rotationSpace=BCS_ParentBoneSpace
    translationMode=BMM_Replace
    translationSpace=BCS_BoneSpace
    alphaBoolBlend {blendOption=Linear}
    boneToModify {boneName=attach_backpack}
    translation {y=3}
   }
   n7 = Control Rig!(Source=Component To Local(ComponentPose=@n6), Alpha, crouch:bool=false) -> (Pose) {
    alphaBoolBlend {blendOption=Linear}
    controlRigAssetReference {blueprintRigClass=&/M_MED_TacticKale_CtrlRig.M_MED_TacticKale_CtrlRig_C}
   }
   Output Pose(Result=@n7) {layerGroup=None}
  ]}
 EventGraph {nodes [
   Event BlueprintUpdateAnimation()
   n1 = Get bIsCrouching()
   > Set notCrouch(notCrouch:double="To Float (Boolean)"(InBool=NOT Boolean(A=@n1))) -> (Output_Get:double)
   n2 = Event OnCharacterCustomizationLoaded()
   Branch(execute=[@n2.then, @n3.then], Condition=Get bIsInFrontEnd())
   > Set Vest Off(Vest Off:bool=CharacterItemDefinitionHasAnyMetaTag(Tags=Make Array("[0]"=(TagName="Cosmetics.MetaTags.TacticKale")))) -> (Output_Get:bool)
   n3 = Event BlueprintInitializeAnimation()
  ]}
]
variables [
 notCrouch {type=number}
 Return Value Location {type=Vector}
 Vest Off {type=boolean}
]
